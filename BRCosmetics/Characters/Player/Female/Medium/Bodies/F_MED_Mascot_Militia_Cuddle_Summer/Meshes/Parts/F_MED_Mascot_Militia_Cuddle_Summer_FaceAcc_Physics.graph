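# Physics asset for the Cuddle Summer mascot's face accessory mesh, defining capsule bodies and joints for ponytail and hair physics.
# A large spine_05 capsule represents the upper body, and each small base hair body joins to it with free swing and twist.
# Swing-limited, nearly twist-locked dyn bodies hang from those bases: a six-link ponytail chain plus single links for three front and two left/right hair pieces.
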PhysicsAsset /BRCosmetics/Characters/Player/Female/Medium/Bodies/F_MED_Mascot_Militia_Cuddle_Summer/Meshes/Parts/F_MED_Mascot_Militia_Cuddle_Summer_FaceAcc_Physics
bodies {
 base_front_hair [{shapeType=Capsule, length=0, radius=0.5, shapeName=base_front_hair_capsule}]
 base_front_hair_l [{shapeType=Capsule, length=0, radius=0.5, shapeName=base_front_hair_l_capsule}]
 base_front_hair_r [{shapeType=Capsule, length=0, radius=0.5, shapeName=base_front_hair_r_capsule}]
 base_hair_l [{shapeType=Capsule, length=0.505, radius=0.505, shapeName=base_hair_l_capsule, rotation=<rot 0 -90 0>}]
 base_hair_r [{shapeType=Capsule, length=0.505, radius=0.505, shapeName=base_hair_r_capsule, rotation=<rot 0 -90 0>}]
 base_ponytail [{shapeType=Capsule, length=0, radius=1, shapeName=base_ponytail_capsule}]
 dyn_front_hair [{shapeType=Capsule, length=1, radius=1, shapeName=dyn_front_hair_capsule, center=<2.5 0 0>, rotation=<rot 90 0 0>}]
 dyn_front_hair_l [{shapeType=Capsule, length=1, radius=1, shapeName=dyn_front_hair_l_capsule, center=<2.5 0 0>, rotation=<rot 90 0 0>}]
 dyn_front_hair_r [{shapeType=Capsule, length=1, radius=1, shapeName=dyn_front_hair_r_capsule, center=<2.5 0 0>, rotation=<rot 90 0 0>}]
 dyn_hair_l [{shapeType=Capsule, length=1, radius=1, shapeName=dyn_hair_l_capsule, center=<2.5 0 0>, rotation=<rot 90 0 0>}]
 dyn_hair_r [{shapeType=Capsule, length=1, radius=1, shapeName=dyn_hair_r_capsule, center=<2.5 0 0>, rotation=<rot 90 0 0>}]
 dyn_ponytail_00 [{shapeType=Capsule, length=0.5, radius=2, shapeName=dyn_ponytail_00_capsule, rotation=<rot -90 0 0>}]
 dyn_ponytail_01 [{shapeType=Capsule, length=2, radius=4, shapeName=dyn_ponytail_01_capsule, center=<2 0 0>, rotation=<rot 90 0 0>}]
 dyn_ponytail_02 [{shapeType=Capsule, length=2, radius=4, shapeName=dyn_ponytail_02_capsule, center=<4 0 0>, rotation=<rot 90 0 0>}]
 dyn_ponytail_03 [{shapeType=Capsule, length=2, radius=4, shapeName=dyn_ponytail_03_capsule, center=<4 0 0>, rotation=<rot 90 0 0>}]
 dyn_ponytail_04 [{shapeType=Capsule, length=1, radius=4, shapeName=dyn_ponytail_04_capsule, center=<4 0 0>, rotation=<rot 90 0 0>}]
 dyn_ponytail_05 [{shapeType=Capsule, length=2, radius=4, shapeName=dyn_ponytail_05_capsule, center=<6 0 0>, rotation=<rot 90 0 0>}]
 spine_05 [{shapeType=Capsule, length=12, radius=14.5, shapeName=spine_05_capsule, center=<-3.4333 0 0>, rotation=<rot 90 0 0>}]
}
constraints [
 {bone1Name=dyn_ponytail_00, bone2Name=base_ponytail, swing1LimitDegrees=10, swing2LimitDegrees=10, twistLimitDegrees=1}
 {
  bone1Name=base_ponytail
  bone2Name=spine_05
  swing1LimitDegrees=45
  swing1Motion=Free
  swing2LimitDegrees=45
  swing2Motion=Free
  twistLimitDegrees=45
  twistMotion=Free
 }
 {bone1Name=dyn_ponytail_01, bone2Name=dyn_ponytail_00, swing1LimitDegrees=30, swing2LimitDegrees=45, twistLimitDegrees=1}
 {bone1Name=dyn_ponytail_02, bone2Name=dyn_ponytail_01, swing1LimitDegrees=35, swing2LimitDegrees=35, twistLimitDegrees=1}
 {bone1Name=dyn_ponytail_03, bone2Name=dyn_ponytail_02, swing1LimitDegrees=30, swing2LimitDegrees=20, twistLimitDegrees=1}
 {bone1Name=dyn_ponytail_04, bone2Name=dyn_ponytail_03, swing1LimitDegrees=30, swing2LimitDegrees=20, twistLimitDegrees=1}
 {bone1Name=dyn_ponytail_05, bone2Name=dyn_ponytail_04, swing1LimitDegrees=30, swing2LimitDegrees=20, twistLimitDegrees=1}
 {
  bone1Name=base_front_hair_l
  bone2Name=spine_05
  swing1LimitDegrees=45
  swing1Motion=Free
  swing2LimitDegrees=45
  swing2Motion=Free
  twistLimitDegrees=45
  twistMotion=Free
 }
 {bone1Name=dyn_front_hair_l, bone2Name=base_front_hair_l, swing1LimitDegrees=15, swing2LimitDegrees=15, twistLimitDegrees=1}
 {
  bone1Name=base_front_hair_r
  bone2Name=spine_05
  swing1LimitDegrees=45
  swing1Motion=Free
  swing2LimitDegrees=45
  swing2Motion=Free
  twistLimitDegrees=45
  twistMotion=Free
 }
 {bone1Name=dyn_front_hair_r, bone2Name=base_front_hair_r, swing1LimitDegrees=15, swing2LimitDegrees=15, twistLimitDegrees=1}
 {
  bone1Name=base_front_hair
  bone2Name=spine_05
  swing1LimitDegrees=45
  swing1Motion=Free
  swing2LimitDegrees=45
  swing2Motion=Free
  twistLimitDegrees=45
  twistMotion=Free
 }
 {bone1Name=dyn_front_hair, bone2Name=base_front_hair, swing1LimitDegrees=15, swing2LimitDegrees=15, twistLimitDegrees=1}
 {bone1Name=dyn_hair_l, bone2Name=base_hair_l, swing1LimitDegrees=20, swing2LimitDegrees=30, twistLimitDegrees=1}
 {
  bone1Name=base_hair_l
  bone2Name=spine_05
  swing1LimitDegrees=45
  swing1Motion=Free
  swing2LimitDegrees=45
  swing2Motion=Free
  twistLimitDegrees=45
  twistMotion=Free
 }
 {
  bone1Name=base_hair_r
  bone2Name=spine_05
  swing1LimitDegrees=45
  swing1Motion=Free
  swing2LimitDegrees=45
  swing2Motion=Free
  twistLimitDegrees=45
  twistMotion=Free
 }
 {bone1Name=dyn_hair_r, bone2Name=base_hair_r, swing1LimitDegrees=20, swing2LimitDegrees=30, twistLimitDegrees=1}
]
