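# Physics asset for the left front face part of the female Cuddle Summer mascot character, setting up physics for its dangling left ponytail bone chain.
# Six small capsule bodies form a chain from the ponytail base, linked by constraints with swing limits that widen toward the larger tip and very tight twist.
# Capsules on the neck, left clavicle and left upper arm, none constrained to the chain, approximate the body around it, apparently as colliders.
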PhysicsAsset /BRCosmetics/Characters/Player/Female/Medium/Bodies/F_MED_Mascot_Militia_Cuddle_Summer/Meshes/Parts/F_MED_Militia_Cuddle_Summer_Face_Front_L_Physics
bodies {
 base_ponytail_l [{shapeType=Capsule, length=0.505, radius=0.505, shapeName=base_ponytail_l_capsule, rotation=<rot 0 -90 0>}]
 clavicle_l [
  {shapeType=Capsule, length=8, radius=7, shapeName=clavicle_l_capsule, center=<6 -1 -1>, rotation=<rot 0 -90 90>}
  {shapeType=Capsule, length=21.75, radius=5, shapeName=clavicle_r_capsule, center=<12 5 1.2>}
  {shapeType=Capsule, length=8, radius=7, shapeName=clavicle_r_capsule, center=<6 1.2 -5.2628>, rotation=<rot 0 90 90>}
  {shapeType=Capsule, length=21.75, radius=7, shapeName=clavicle_r_capsule, center=<6.4887 -9.204 1.2>}
 ]
 dyn_ponytail_00_l [{shapeType=Capsule, length=0, radius=0.6, shapeName=dyn_ponytail_00_l_capsule, center=<1 0 0>, rotation=<rot 90 0 0>}]
 dyn_ponytail_01_l [{shapeType=Capsule, length=0, radius=0.6, shapeName=dyn_ponytail_01_l_capsule, center=<1 0 0>, rotation=<rot 90 0 0>}]
 dyn_ponytail_02_l [{shapeType=Capsule, length=0, radius=0.6, shapeName=dyn_ponytail_02_l_capsule, center=<1 0 0>, rotation=<rot 90 0 0>}]
 dyn_ponytail_03_l [{shapeType=Capsule, length=0, radius=0.6, shapeName=dyn_ponytail_03_l_capsule, center=<1 0 0>, rotation=<rot 90 0 0>}]
 dyn_ponytail_04_l [{shapeType=Capsule, length=0, radius=0.7, shapeName=dyn_ponytail_04_l_capsule, center=<1 0 0>, rotation=<rot 90 0 0>}]
 dyn_ponytail_05_l [{shapeType=Capsule, length=4, radius=1.5, shapeName=dyn_ponytail_05_l_capsule, center=<3.5 0 0>, rotation=<rot 90 0 0>}]
 neck_01 [{shapeType=Capsule, length=15, radius=5, shapeName=clavicle_r_capsule, rotation=<rot 90 0 0>}]
 upperarm_l [{shapeType=Capsule, length=10, radius=3, shapeName=upperarm_l_capsule, center=<10 0 0>, rotation=<rot 90 0 0>}]
}
constraints [
 {bone1Name=dyn_ponytail_01_l, bone2Name=dyn_ponytail_00_l, swing1LimitDegrees=15, swing2LimitDegrees=15, twistLimitDegrees=1}
 {bone1Name=dyn_ponytail_02_l, bone2Name=dyn_ponytail_01_l, swing1LimitDegrees=20, swing2LimitDegrees=20, twistLimitDegrees=1}
 {bone1Name=dyn_ponytail_03_l, bone2Name=dyn_ponytail_02_l, swing1LimitDegrees=20, swing2LimitDegrees=20, twistLimitDegrees=1}
 {bone1Name=dyn_ponytail_04_l, bone2Name=dyn_ponytail_03_l, swing1LimitDegrees=20, swing2LimitDegrees=20, twistLimitDegrees=1}
 {bone1Name=dyn_ponytail_05_l, bone2Name=dyn_ponytail_04_l, swing1LimitDegrees=30, swing2LimitDegrees=45, twistLimitDegrees=1}
 {bone1Name=dyn_ponytail_00_l, bone2Name=base_ponytail_l, swing1LimitDegrees=15, swing2LimitDegrees=10, twistLimitDegrees=1}
]
